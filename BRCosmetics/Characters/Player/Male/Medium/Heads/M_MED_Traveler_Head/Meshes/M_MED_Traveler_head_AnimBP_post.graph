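# Post-process AnimBlueprint for the male medium Traveler head that corrects facial bones on the incoming pose during emotes and custom animations.
# A chain of additive Transform Bone nodes, weighted by customAnimAlpha, adjusts the teeth, tongue, mid brows and inner cheeks.
# Extra mouth tweaks apply while either Epic Sax Guy montage is active on the owner mesh's anim instance.
# The update event sets these alphas, plus an emotingAlpha the AnimGraph never reads.
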AnimBlueprint /BRCosmetics/Characters/Player/Male/Medium/Heads/M_MED_Traveler_Head/Meshes/M_MED_Traveler_head_AnimBP_post
target_skeleton=/Game/Characters/Player/Common/Fortnite_Base_Head/Fortnite_Base_Head_Export_Skeleton
graphs [
 AnimGraph {nodes [
   AnimGraphNode_SubInput_0 = Input Pose()
   n1 = Get customAnimAlpha() -> (customAnimAlpha:double)
   n2 = "Transform (Modify) Bone - Bone: teeth_upper"(ComponentPose=Local To Component(LocalPose=@AnimGraphNode_SubInput_0), Alpha=@n1, Translation="0,-1,0", Rotation="0,0,0", Scale="1,1,1") {
    scaleMode=BMM_Additive
    scaleSpace=BCS_BoneSpace
    translationMode=BMM_Additive
    translationSpace=BCS_BoneSpace
    alphaBoolBlend {blendOption=Linear}
    boneToModify {boneName=teeth_upper}
    translation {y=-1}
   }
   n3 = "Transform (Modify) Bone - Bone: teeth_lower"(ComponentPose=@n2, Alpha=@n1, Translation="0,-1,0", Rotation="0,0,0", Scale="0.9,1,1") {
    scaleMode=BMM_Additive
    translationMode=BMM_Additive
    translationSpace=BCS_BoneSpace
    alphaBoolBlend {blendOption=Linear}
    boneToModify {boneName=teeth_lower}
    scale {x=0.9}
    translation {y=-1}
   }
   n4 = "Transform (Modify) Bone - Bone: tongue"(ComponentPose=@n3, Alpha=@n1, Translation="0,-1,0", Rotation="0,0,0", Scale="1,1,1") {
    translationMode=BMM_Additive
    translationSpace=BCS_BoneSpace
    alphaBoolBlend {blendOption=Linear}
    boneToModify {boneName=tongue}
    translation {y=-1}
   }
   n5 = Get isPlayingSaxEmote() -> (isPlayingSaxEmote:double)
   n6 = "Transform (Modify) Bone - Bone: teeth_lower"(ComponentPose=@n4, Alpha=@n5, Translation="0,0,0", Rotation="0,0,0", Scale="0.7,1,1") {
    scaleMode=BMM_Additive
    scaleSpace=BCS_BoneSpace
    translationSpace=BCS_BoneSpace
    alphaBoolBlend {blendOption=Linear}
    boneToModify {boneName=teeth_lower}
    scale {x=0.7}
   }
   n7 = "Transform (Modify) Bone - Bone: tongue"(ComponentPose=@n6, Alpha=@n5, Translation="0,0,0", Rotation="0,0,0", Scale="1,0.8,1") {scaleMode=BMM_Additive, translationSpace=BCS_BoneSpace, alphaBoolBlend {blendOption=Linear}, boneToModify {boneName=tongue}, scale {y=0.8}}
   n8 = "Transform (Modify) Bone - Bone: teeth_upper"(ComponentPose=@n7, Alpha=@n5, Translation="0,1,0", Rotation="0,0,0", Scale="1,1,1") {
    scaleMode=BMM_Additive
    scaleSpace=BCS_BoneSpace
    translationMode=BMM_Additive
    translationSpace=BCS_BoneSpace
    alphaBoolBlend {blendOption=Linear}
    boneToModify {boneName=teeth_upper}
    translation {y=1}
   }
   n9 = Get customAnimAlpha()
   n10 = "Transform (Modify) Bone - Bone: R_brow_mid"(ComponentPose=@n8, Alpha=@n9, Translation="0,0.5,0", Rotation="0,0,0", Scale="1,1,1") {
    scaleMode=BMM_Additive
    scaleSpace=BCS_BoneSpace
    translationMode=BMM_Additive
    translationSpace=BCS_BoneSpace
    alphaBoolBlend {blendOption=Linear}
    boneToModify {boneName=R_brow_mid}
    translation {y=0.5}
   }
   n11 = "Transform (Modify) Bone - Bone: R_cheek_inner"(ComponentPose=@n10, Alpha=@n9, Translation="0,0.5,0", Rotation="0,0,0", Scale="1,1,1") {
    scaleMode=BMM_Additive
    scaleSpace=BCS_BoneSpace
    translationMode=BMM_Additive
    translationSpace=BCS_BoneSpace
    alphaBoolBlend {blendOption=Linear}
    boneToModify {boneName=R_cheek_inner}
    translation {y=0.5}
   }
   n12 = "Transform (Modify) Bone - Bone: L_brow_mid"(ComponentPose=@n11, Alpha=@n9, Translation="0,0.5,0", Rotation="0,0,0", Scale="1,1,1") {
    scaleMode=BMM_Additive
    scaleSpace=BCS_BoneSpace
    translationMode=BMM_Additive
    translationSpace=BCS_BoneSpace
    alphaBoolBlend {blendOption=Linear}
    boneToModify {boneName=L_brow_mid}
    translation {y=0.5}
   }
   n13 = "Transform (Modify) Bone - Bone: L_cheek_inner"(ComponentPose=@n12, Alpha=@n9, Translation="0,0.5,0", Rotation="0,0,0", Scale="1,1,1") {
    scaleMode=BMM_Additive
    scaleSpace=BCS_BoneSpace
    translationMode=BMM_Additive
    translationSpace=BCS_BoneSpace
    alphaBoolBlend {blendOption=Linear}
    boneToModify {boneName=L_cheek_inner}
    translation {y=0.5}
   }
   Output Pose(Result=Component To Local(ComponentPose=@n13)) {layerGroup=None}
  ]}
 EventGraph {nodes [
   Event BlueprintUpdateAnimation()
   > n1 = Sequence()
   n2 = Get bIsPlayingEmote()
   > Set emotingAlpha(emotingAlpha:double="To Float (Boolean)"(InBool=@n2)) -> (Output_Get:double)
   n3 = Get bCanPlayCustomAnimations() -> (bCanPlayCustomAnimations:bool)
   > Set customAnimAlpha(customAnimAlpha:double="To Float (Boolean)"(InBool=OR Boolean(A=@n2, B=@n3))) -> (Output_Get:double)
   n4 = GetAnimInstance(self=GetOwnerSkeletalMeshComponent())
   Is Valid(exec=@n1.then_1, InputObject=@n4)
   n5 = Montage_IsActive!(
    self:Anim Instance Object Reference=@n4
    Montage:Anim Montage Object Reference=/BRCosmetics/Animation/Game/MainPlayer/Emotes/Epic_Sax_Guy/Emote_EpicSaxGuy_CMM_M1
   ) -> (ReturnValue:bool)
   n6 = Montage_IsActive(self=@n4, Montage=/BRCosmetics/Animation/Game/MainPlayer/Emotes/Epic_Sax_Guy/Emote_EpicSaxGuy_CMF_M)
   > Set isPlayingSaxEmote(isPlayingSaxEmote:double=SelectFloat(A=1, B=0, bPickA=OR Boolean(A=@n5, B=@n6))) -> (Output_Get:double)
   TryGetPawnOwner()
  ]}
]
variables [
 customAnimAlpha {type=number}
 emotingAlpha {type=number}
 isPlayingSaxEmote {type=number}
]
